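# Masked, two-sided material for the Nanomist Vortex, a tornado of spline-based ribbons with dithered pixel depth offset.
# World position offset builds ribbons from UV-encoded spline points, using curve-atlas profile, noise and camera-distance rotation, collapsing extra ribbons on mobile.
# Flowmapped swirl normals and warped panning coordinates feed MF_LackLuster_Nanomist for the metallic mist look.
# Further branches compute Fresnel hue-shifted iridescence, storm lightning (zero on mobile) and Blueprint-timed fade progression.
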Material /Game/NeverCook/FX/Vortex/Materials/M_NanomistVortex
graph {
 properties {
  bUseMaterialAttributes=true
  blendMode=BLEND_Masked
  ditherOpacityMask=true
  pixelDepthOffsetMode=PDOM_Legacy
  twoSided=true
  usedShadingModels=Default Lit
 }
 nodes [
  n1 = MaterialFunctionCall()
  n2 = FeatureLevelSwitch(Default=@n1, ES3_1=@n1)
  n3 = CollectionParameter() {
   collection=&FortniteMaterialParameters
   parameterName=SunAndMoonModelDirectionalVector
  }
  n4 = Clamp(None=DotProduct(A=ComponentMask(None=@n3) {b=true, g=true, r=true}, B=ReflectionVectorWS()))
  n5 = ScalarParameter() {defaultValue=2, group=Color, parameterName=Spec Highlight Power}
  n6 = LinearInterpolate(
   A=@n2
   B=Multiply(A=ScalarParameter() {defaultValue=2, desc=Intensity of the Fresnel effect, group=Color, parameterName=Fresnel Brightness}, B=@n2)
   Alpha=Multiply(A=Power(Base=@n4, Exp=@n5) {constExponent=20}, B=ScalarParameter() {defaultValue=1.5, group=Color, parameterName=Spec Intense})
  )
  Add(A=@n6, B=NamedRerouteUsage())
  Constant() {r=1}
  n7 = ScalarParameter() {defaultValue=1, group=Color, parameterName=InscatterBrightness}
  n8 = Subtract(A=NamedRerouteUsage(), B=Constant())
  n9 = Constant() {r=2}
  LinearInterpolate(
   A=Multiply(A=@n7, B=ScalarParameter() {defaultValue=0.1, group=Color, parameterName=interscatter_Top}) {constB=0.7}
   B=@n7
   Alpha=Saturate(None=Divide(A=@n8, B=@n9))
  )
  n10 = ScalarParameter() {parameterName=Normal Flatness}
  n11 = Fresnel(Normal=VertexNormalWS()) {baseReflectFraction=0}
  n12 = MaterialFunctionCall(
   Normal=NamedRerouteUsage()
   Flatness=LinearInterpolate(A=ScalarParameter() {defaultValue=0.95, parameterName=Normal Flatness Swirl}, Alpha=@n11) {constB=0}
  ) {materialFunction=&FlattenNormal}
  LinearInterpolate(
   A=MaterialFunctionCall(Normal=NamedRerouteUsage(), Flatness=LinearInterpolate(A=@n10, Alpha=@n11) {constB=0}) {materialFunction=&FlattenNormal}.Result
   B=@n12.Result
   Alpha=ScalarParameter() {defaultValue=1, parameterName=Normal Bias}
  )
  n13 = Normalize(VectorInput=@n12.Result)
  n14 = SetMaterialAttributes!(
   MaterialAttributes
   Base Color=Constant()
   Metallic
   Specular=Constant()
   Roughness
   Normal=@n13
   Emissive Color
   Opacity Mask=NamedRerouteUsage()
   World Position Offset
   Pixel Depth Offset=Multiply(
    A=NamedRerouteUsage()
    B=MaterialFunctionCall() {materialFunction=&DitherTemporalAA}.Result
   )
  ) {attributeSetTypes [
    #1
    #2
    #3
    #4
    #5
    #6
    #7
    #8
    #9
   ]}
  n15 = NamedRerouteUsage()
  n16 = ComponentMask(None=NamedRerouteUsage()) {r=true}
  n17 = NamedRerouteUsage()
  n18 = MaterialFunctionCall() {materialFunction=&BoundingBoxBased_0-1_UVW}
  n19 = Saturate(None=Max(A=Multiply(A=@n16) {constB=0.5}, B=Multiply(A=@n17, B=@n18.B)))
  n20 = GetMaterialAttributes(None=MaterialFunctionCall!(
    bBlendMaterialAttributes
    Override Panning Coordinates 2=NamedRerouteUsage()
    MA_In=@n14
    UVsOverride
    Override Panning Coordinates=@n15
    Phase Multiplier=@n19
    isPhased=@n19
    BlendMaterialAttributesAlpha
    BaseColorOverride
    WorldPositionOverride
    VertexNormalOverride
    SourceNormalsStrengthOverride
   ) {materialFunction=&/MF_LackLuster_Nanomist}.MA Out) {attributeGetTypes [#5]}
  n21 = ScalarParameter() {defaultValue=0.5, parameterName=Normal Blend}
  MaterialFunctionCall!(
   Material=SetMaterialAttributes!(MaterialAttributes=@n20.MaterialAttributes, Opacity Mask=NamedRerouteUsage(), Normal=LinearInterpolate(A=@n20.Normal, B=@n13, Alpha=@n21)) {attributeSetTypes [#7, #5]}
   Mask
   WorldPositionOffset=NamedRerouteUsage()
  ) {materialFunction=/Engine/Functions/MaterialLayerFunctions/MatLayerBlend_OverrideWorldPositionOffset}
  n22 = Multiply(A=Subtract(A=WorldPosition().XYZ, B=ActorPositionWS()), B=Constant3Vector() {constant=<rgba 1 1 0 0>})
  MaterialFunctionCall(
   BaseNormal=Transform(None=Normalize(VectorInput=@n22)) {transformSourceType=TRANSFORMSOURCE_World, transformType=TRANSFORM_Tangent}
   AdditionalNormal=NamedRerouteUsage()
  ) {materialFunction=&BlendAngleCorrectedNormals}
  n23 = Transform(None=Subtract(A=CameraPositionWS(), B=ActorPositionWS())) {transformSourceType=TRANSFORMSOURCE_World, transformType=TRANSFORM_Local}
  n24 = Saturate(None=Multiply(A=ComponentMask(None=@n23) {b=true}, B=Constant() {r=0.1}) {constB=0.01})
  Multiply(A=@n24, B=@n24)
  n25 = Constant() {r=0.25}
  n26 = Add(A=@n25, B=Constant() {r=0.25})
  n27 = NamedRerouteUsage()
  n28 = TextureCoordinate()
  n29 = ScalarParameter() {defaultValue=0.9, group=Opacity, parameterName=RibbonSidesOpacityBottom}
  n30 = ScalarParameter() {defaultValue=0.5, group=Opacity, parameterName=RibbonSidesOpacityTop}
  n31 = Power(Base=ComponentMask(None=TextureCoordinate()) {r=true}, Exp=Constant() {r=0.2})
  n32 = Constant() {r=1.09}
  n33 = Multiply(A=NamedRerouteUsage(), B=Constant() {r=-0.6})
  n34 = OneMinus(None=NamedRerouteUsage())
  n35 = Saturate(None=InverseLinearInterpolate(A=Constant() {r=0.0762}, B=Add(A=@n32, B=@n33), Value=Multiply(A=@n34, B=@n34)))
  n36 = ComponentMask(None=TextureCoordinate()) {r=true}
  n37 = NamedRerouteUsage()
  n38 = Constant() {r=0.3333}
  n39 = Subtract(A=Divide(A=Subtract(A=@n36, B=@n35), B=Constant() {r=0.05}), B=Divide(A=Subtract(A=@n37, B=@n38), B=Constant() {r=0.3214}))
  n40 = Subtract(A=OneMinus(None=@n36), B=Constant() {r=0.2543})
  n41 = Constant() {r=0.0576}
  n42 = Subtract(A=@n37, B=Constant() {r=0.5619})
  n43 = Constant() {r=0.15}
  n44 = Saturate(None=Subtract(A=Divide(A=@n40, B=@n41), B=Divide(A=@n42, B=@n43)))
  n45 = Multiply(
   A=SphereMask(A=ComponentMask(None=@n28) {g=true}, B=Constant() {r=0.5}, Radius=Constant() {r=0.5}, Hardness=LinearInterpolate(A=@n29, B=@n30, Alpha=@n31))
   B=Saturate(None=Multiply(A=@n39, B=@n44))
  )
  Multiply(A=Saturate(None=InverseLinearInterpolate(A=@n25, B=@n26, Value=@n27)), B=@n45)
  n46 = TextureCoordinate() {coordinateIndex=1}
  n47 = TextureCoordinate() {coordinateIndex=2}
  n48 = ComponentMask(None=TextureCoordinate() {coordinateIndex=1}) {g=true}
  NamedRerouteDeclaration(None=MaterialFunctionCall(R=ComponentMask(None=@n46) {r=true}, G=ComponentMask(None=@n47) {r=true}, B=OneMinus(None=@n48)) {materialFunction=&AppendMany}.RGB) {name=SplinePointPosition, nodeColor=<rgba 0.3828 0 1 1>}
  n49 = PreSkinnedPosition()
  n50 = NamedRerouteUsage()
  NamedRerouteDeclaration(None=Normalize(VectorInput=Subtract(A=@n49, B=@n50))) {name=SplinePointToVertexVector, nodeColor=<rgba 0.8281 0 1 1>}
  n51 = TextureCoordinate()
  n52 = ComponentMask(None=Power(Base=@n51, Exp=ScalarParameter() {defaultValue=0.333, group=UVs, parameterName=Ribbon_U_Pow})) {r=true}
  n53 = Round(None=NamedRerouteUsage())
  n54 = AppendVector(A=LinearInterpolate(A=@n52, B=OneMinus(None=@n52), Alpha=@n53), B=ComponentMask(None=@n51) {g=true})
  n55 = AppendVector(
   A=ScalarParameter() {defaultValue=5, group=UVs, parameterName=TexTilingV}
   B=ScalarParameter() {defaultValue=0.7, group=UVs, parameterName=TexTilingU}
  )
  n56 = Multiply(A=@n53) {constB=-2}
  n57 = Round(None=NamedRerouteUsage())
  n58 = Constant() {r=2}
  n59 = Multiply(A=Add(A=Multiply(A=@n57, B=@n58) {constB=11}), B=ScalarParameter() {defaultValue=-0.07, group=UVs, parameterName=TexScrollingSpeed})
  n60 = Multiply(A=NamedRerouteUsage(), B=ScalarParameter() {defaultValue=-0.05, group=UVs, parameterName=TexScrollingSpeedVertexColorOffset})
  n61 = Multiply(A=Time(), B=Multiply(A=Subtract(A=@n56) {constB=-1}, B=Add(A=@n59, B=@n60)))
  n62 = Constant()
  NamedRerouteDeclaration(None=Add(A=Multiply(A=@n54, B=@n55), B=AppendVector(A=@n61, B=@n62))) {name=UVPanning, nodeColor=<rgba 1 0.9375 0 1>}
  n63 = Transform(None=NamedRerouteUsage()) {transformSourceType=TRANSFORMSOURCE_World, transformType=TRANSFORM_Tangent}
  n64 = ScalarParameter() {defaultValue=1, desc=Normal, parameterName=NormalMipBias}
  n65 = ScalarParameter() {defaultValue=5, desc=Normal, parameterName=NormalMipBiasB}
  n66 = ComponentMask(None=TextureCoordinate()) {r=true}
  n67 = ScalarParameter() {defaultValue=0.3, desc=Normal, parameterName=NormalStrength}
  n68 = Multiply(
   A=TextureSample!(UVs=NamedRerouteUsage(), Tex, Bias=LinearInterpolate(A=@n64, B=@n65, Alpha=@n66), Apply View MipBias) {mipValueMode=TMVM_MipBias, samplerType=SAMPLERTYPE_Normal}.RGB
   B=AppendVector(A=AppendVector(A=@n67, B=@n67), B=Constant() {r=1})
  )
  NamedRerouteDeclaration(None=MaterialFunctionCall(BaseNormal=VertexInterpolator(VS=@n63).PS, AdditionalNormal=Normalize(VectorInput=@n68)) {materialFunction=&BlendAngleCorrectedNormals}.Result) {name=Normal_Out, nodeColor=<rgba 0.2891 0 1 1>}
  n69 = CurveAtlasRowParameter(CurveTime=MaterialFunctionCall() {materialFunction=&MF_Fresnel_FX}.Fresnel) {
   atlas=/Game/TechArt/Seasonal/LackLuster/Curves/CA_Lackluster
   curve=/Game/TechArt/Seasonal/LackLuster/Curves/C_LackLuster_Spectrum_03
   defaultValue=3
   parameterName=ChromaCurve
  }
  n70 = MaterialFunctionCall(Exponent=ScalarParameter() {defaultValue=1.5, group=HueShift, parameterName=FresnelHueShiftExponent}) {materialFunction=&MF_Fresnel_FX}
  n71 = Multiply(A=@n70.Fresnel Inverted Double Sided, B=ScalarParameter() {group=HueShift, parameterName=HueShiftScale})
  n72 = TextureCoordinate()
  n73 = AppendVector(A=ScalarParameter() {defaultValue=5, parameterName=U_Scale}, B=ScalarParameter() {defaultValue=0.5, parameterName=V_Scale})
  n74 = ScalarParameter() {defaultValue=0.1, parameterName=PanSpeedChrome}
  n75 = Constant()
  n76 = TextureSample!(UVs=Panner(Coordinate=Multiply(A=@n72, B=@n73), Speed=AppendVector(A=@n74, B=@n75)), Tex, Apply View MipBias) {samplerType=SAMPLERTYPE_LinearColor, texture=/Game/TechArt/Seasonal/LackLuster/Textures/T_Lackluster_Streaks}
  n77 = CurveAtlasRowParameter(CurveTime=Frac(None=Add(A=@n76.R, B=@n70.Fresnel Double Sided))) {
   atlas=&CA_Lackluster
   curve=/Game/TechArt/Seasonal/LackLuster/Curves/C_Lackluster_Reflection
   parameterName=ReflectionCurve
  }
  n78 = Add(
   A=Multiply(A=@n69, B=ScalarParameter() {defaultValue=1.5, parameterName=ChromaCurveScale})
   B=MaterialFunctionCall(Hue Shift Percentage=Frac(None=@n71), Texture=@n77) {materialFunction=&HueShift}.Result
  )
  n79 = MaterialFunctionCall(Exponent=ScalarParameter() {defaultValue=0.9, parameterName=FinalMaskExponent}) {materialFunction=&MF_Fresnel_FX}
  NamedRerouteDeclaration(None=Add(A=@n6, B=Multiply(A=@n78, B=@n79.Fresnel Inverted Double Sided))) {name=Emissive_Out, nodeColor=<rgba 0.4531 0 1 1>}
  n80 = TextureObject() {samplerType=SAMPLERTYPE_Normal, texture=/Game/TechArt/Seasonal/LackLuster/Textures/T_Lackluster_Swirl_N}
  n81 = ScalarParameter() {defaultValue=3, parameterName=SwirlTileX}
  n82 = ScalarParameter() {defaultValue=0.25, parameterName=SwirlTileY}
  n83 = Add(
   A=Multiply(
    A=TextureCoordinate()
    B=MaterialFunctionCall(X=@n81, Y=@n82) {materialFunction=&MakeFloat2}.Result
   )
   B=NamedRerouteUsage()
  )
  n84 = Panner(Coordinate=@n83, Time=NamedRerouteUsage()) {speedX=-0.2}
  n85 = ComponentMask(None=ConstantBiasScale(None=TextureSample(UVs=@n84) {samplerType=SAMPLERTYPE_LinearColor, texture=/Game/Effects/Fort_Effects/Textures/Plasma/T_Turbulence_AdditiveWarp_Linear}.RGB) {bias=-0.5, scale=2}) {g=true, r=true}
  n86 = Divide(A=@n85, B=ScalarParameter() {defaultValue=5, parameterName=FlowmapSpeedDivider}) {constB=3.5}
  n87 = NamedRerouteUsage()
  NamedRerouteDeclaration(None=MaterialFunctionCall(
    BaseNormal=MaterialFunctionCall!(Volume Texture=@n80, UV=@n83, Velocity=@n86, Override Time=@n87, Mip Level) {materialFunction=/Engine/Functions/Engine_MaterialFunctions02/Texturing/Flowmaps_2D}.Result RGB
    AdditionalNormal=MaterialFunctionCall(Volume Texture=@n80, UV=@n83, Velocity=@n86, Override Time=@n87) {materialFunction=&Flowmaps_2D}.Result RGB
   ) {materialFunction=&BlendAngleCorrectedNormals}.Result) {name=Swirl Normals, nodeColor=<rgba 0 0.3437 1 1>}
  n88 = ObjectPositionWS()
  n89 = WorldPosition()
  n90 = Add(A=NamedRerouteUsage(), B=NamedRerouteUsage())
  n91 = NamedRerouteUsage()
  n92 = Transform(None=Add(A=Add(A=@n90, B=@n91), B=NamedRerouteUsage())) {transformSourceType=TRANSFORMSOURCE_Local}
  n93 = NamedRerouteUsage()
  n94 = ScalarParameter() {defaultValue=11, parameterName=RibbonsCountMobile}
  n95 = Divide(A=ShadingPathSwitch(Default=ScalarParameter() {defaultValue=16, parameterName=RibbonsCount}, Forward=@n94, Mobile=@n94)) {constB=16}
  NamedRerouteDeclaration(None=LinearInterpolate(
    A=Subtract(A=@n88, B=@n89.XYZ)
    B=MaterialFunctionCall!(InVector=@n92) {materialFunction=/Game/Athena/Environments/Landscape/Foliage/MF_MakePreciseVec3}.Result
    Alpha=Step(Y=@n93, X=@n95)
   )) {name=WPO_Out, nodeColor=<rgba 0.2188 0 1 1>}
  n96 = TextureSample!(UVs=NamedRerouteUsage(), Tex, Bias=ScalarParameter() {group=PDO, parameterName=MipBias}, Apply View MipBias) {mipValueMode=TMVM_MipBias}
  NamedRerouteDeclaration(None=@n96.RGB) {name=DiffuseTex, nodeColor=<rgba 1 0.3516 0 1>}
  n97 = CameraVectorWS()
  n98 = PixelNormalWS()
  NamedRerouteDeclaration(None=Saturate(None=DotProduct(A=@n97, B=@n98))) {name=NdotC_PS, nodeColor=<rgba 0 0.9531 1 1>}
  n99 = Constant() {r=0.5}
  n100 = Constant()
  n101 = ComponentMask(None=TextureCoordinate()) {r=true}
  NamedRerouteDeclaration(None=Saturate(None=InverseLinearInterpolate(A=@n99, B=@n100, Value=@n101))) {name=VertGradient_Color, nodeColor=<rgba 1 0.8438 0 1>}
  n102 = TextureCoordinate()
  n103 = Multiply(A=Distance(A=ComponentMask(None=@n102) {g=true}, B=Constant() {r=0.5})) {constB=2}
  n104 = Multiply(
   A=Multiply(A=@n103, B=@n103)
   B=Multiply(
    A=MaterialFunctionCall() {materialFunction=&ObjectScale}.Scale Z
    B=ScalarParameter() {defaultValue=0.7, group=PDO, parameterName=pdo}
   )
  )
  n105 = Constant()
  NamedRerouteDeclaration(None=ShadingPathSwitch(Default=Multiply(A=@n104, B=@n96.R), Forward=@n105, Mobile=@n105)) {name=PDO_Out, nodeColor=<rgba 0 0.2266 1 1>}
  n106 = NamedRerouteUsage()
  n107 = TextureCoordinate()
  n108 = Multiply(A=CurveAtlasRowParameter(CurveTime=ComponentMask(None=@n107) {r=true}) {parameterName=ScaleCurve}, B=NamedRerouteUsage()) {constB=0.09}
  n109 = ScalarParameter() {defaultValue=-1, parameterName=GlobalWidthScaleBack}
  n110 = ScalarParameter() {defaultValue=0.7, parameterName=GlobalWidthScaleFront}
  n111 = InverseLinearInterpolate(
   A=ScalarParameter() {defaultValue=0.4, parameterName=Param_2}
   B=ScalarParameter() {defaultValue=0.52, parameterName=Param_13}
   Value=NamedRerouteUsage()
  )
  n112 = Constant() {r=0.4}
  n113 = Constant() {r=0.6}
  n114 = ComponentMask(None=TextureCoordinate()) {r=true}
  n115 = Max(A=Max(A=Saturate(None=@n111), B=NamedRerouteUsage()), B=Saturate(None=InverseLinearInterpolate(A=@n112, B=@n113, Value=@n114)))
  NamedRerouteDeclaration(None=Multiply(A=Multiply(A=@n106, B=@n108), B=LinearInterpolate(A=@n109, B=@n110, Alpha=@n115))) {name=RibbonWidthPosOffset, nodeColor=<rgba 1 0.6797 0 1>}
  n116 = Multiply(A=Time(), B=ScalarParameter() {defaultValue=1, group=Lightning, parameterName=LightningTimeShift})
  n117 = Frac(None=@n116)
  n118 = OneMinus(None=@n117)
  n119 = Clamp(None=Multiply(A=Multiply(A=@n117, B=@n118), B=Constant() {r=1}))
  n120 = PreSkinnedPosition()
  n121 = Ceil(None=@n116)
  n122 = ScalarParameter() {defaultValue=0.1062, group=Lightning, parameterName=LightningTimeShiftOffset}
  n123 = TextureSampleParameter2D(UVs=AppendVector(A=Multiply(A=@n121, B=@n122), B=Constant())) {
   parameterName=LightningTimeline
   samplerType=SAMPLERTYPE_Masks
   texture=/Game/Gadgets/Assets/VinderTech_GliderChute/Textures/T_LightingTimeline_M
  }
  n124 = Add(
   A=Divide(A=ComponentMask(None=@n120) {g=true, r=true}, B=ScalarParameter() {defaultValue=6, group=Lightning, parameterName=LightningScale})
   B=Multiply(A=AppendVector(A=@n123.G, B=@n123.G)) {constB=5.0061}
  )
  n125 = Power(
   Base=TextureSampleParameter2D(UVs=VertexInterpolator(VS=@n124).PS) {parameterName=LightningTexture, samplerType=SAMPLERTYPE_LinearColor, texture=/Game/Athena/SafeZone/Textures/T_Clouds_Lightning}.G
   Exp=Constant() {r=4}
  ) {constExponent=4}
  n126 = Add(A=Time(), B=Constant() {r=0.2})
  n127 = Constant() {r=0.05}
  n128 = CurveAtlasRowParameter(CurveTime=Frac(None=Multiply(A=@n126, B=@n127))) {
   atlas=&AthenaCurveAtlas01
   curve=/Game/Athena/SafeZone/Textures/StormLighting
   defaultValue=18
   parameterName=StormLightningCurve
  }
  n129 = NamedRerouteUsage()
  n130 = Multiply(A=Multiply(A=Multiply(A=@n119, B=@n125), B=VertexInterpolator(VS=@n128.R).PS), B=Multiply(A=@n129, B=@n129))
  n131 = ScalarParameter() {defaultValue=80, parameterName=LightningIntensity}
  n132 = Constant()
  NamedRerouteDeclaration(None=ShadingPathSwitch(Default=Multiply(A=@n130, B=@n131), Forward=@n132, Mobile=@n132)) {name=Lightning, nodeColor=<rgba 0 0.7187 1 1>}
  n133 = Time()
  n134 = ScalarParameter() {defaultValue=1, group=Progression, parameterName=ToggleVisibilityFromBP}
  n135 = Add(A=ScalarParameter() {group=Progression, parameterName=GameTimeFromBP}, B=@n134)
  n136 = ScalarParameter() {defaultValue=1, group=Progression, parameterName=AnimationLengthFromBP}
  n137 = Saturate(None=Divide(A=Subtract(A=@n133, B=@n135), B=Subtract(A=@n136, B=@n134)))
  NamedRerouteDeclaration(None=LinearInterpolate(
    A=CurveAtlasRowParameter(CurveTime=@n137) {defaultValue=5, parameterName=FadeOutCurve}.G
    B=CurveAtlasRowParameter(CurveTime=@n137) {defaultValue=4, parameterName=FadeInCurve}.G
    Alpha=@n134
   )) {name=Progression, nodeColor=<rgba 0.9219 0 1 1>}
  n138 = CameraPositionWS()
  n139 = WorldPosition()
  NamedRerouteDeclaration(None=Normalize(VectorInput=Subtract(A=@n138, B=@n139.XYZ))) {name=CameraToVertexWS, nodeColor=<rgba 0 0.1562 1 1>}
  n140 = TextureCoordinate() {coordinateIndex=2}
  NamedRerouteDeclaration(None=OneMinus(None=ComponentMask(None=@n140) {g=true})) {name=RibbonID, nodeColor=<rgba 0.9922 1 0 1>}
  n141 = VertexNormalWS()
  n142 = NamedRerouteUsage()
  n143 = Add(A=Multiply(A=DotProduct(A=@n141, B=@n142)) {constB=0.5}) {constB=0.5}
  NamedRerouteDeclaration(None=@n143) {name=NdotC_VS, nodeColor=<rgba 0 0.2734 1 1>}
  n144 = CameraPositionWS()
  n145 = ActorPositionWS()
  n146 = Distance(A=Subtract(A=ComponentMask(None=@n144) {g=true, r=true}, B=ComponentMask(None=@n145) {g=true, r=true}), B=Constant())
  n147 = LinearInterpolate(
   A=ScalarParameter() {defaultValue=700, parameterName=InnerRotationBlendStart}
   B=ScalarParameter() {defaultValue=5000, parameterName=InnerRotationBlendStart-Top}
  ) {constAlpha=0}
  n148 = Saturate(None=Divide(A=Subtract(A=@n146, B=@n147), B=ScalarParameter() {defaultValue=700, parameterName=InnerRotationBlendRange}))
  n149 = OneMinus(None=@n148)
  NamedRerouteDeclaration(None=ShaderStageSwitch(PixelShader=VertexInterpolator(VS=@n149).PS, VertexShader=@n149)) {name=InsideTornado, nodeColor=<rgba 0 0.0859 1 1>}
  n150 = TextureObject() {samplerType=SAMPLERTYPE_LinearColor, texture=/Game/NeverCook/FX/Vortex/Textures/T_Lackluster_Streaks_NanomistVortex}
  n151 = Power(
   Base=ComponentMask(None=MaterialFunctionCall(Volume Texture=@n150, UV=@n83, Velocity=@n86, Override Time=@n87) {materialFunction=&Flowmaps_2D}.Result RGB) {r=true}
   Exp=LinearInterpolate(Alpha=NamedRerouteUsage()) {constA=6, constB=0.1}
  ) {constExponent=6}
  n152 = NamedRerouteUsage()
  NamedRerouteDeclaration(None=Multiply(A=Saturate(None=@n151), B=OneMinus(None=@n152))) {name=Swirl Dot, nodeColor=<rgba 0 0.0156 1 1>}
  n153 = Add(A=@n83, B=Multiply(A=ScalarParameter() {defaultValue=0.05, parameterName=Glow Coordinates Warp}, B=@n85))
  n154 = NamedRerouteUsage()
  NamedRerouteDeclaration(None=Panner(Coordinate=Multiply(A=@n153) {constB=2}, Time=@n154) {speedX=-0.2, speedY=0.1}) {name=Panning Coordinates, nodeColor=<rgba 0.7813 0 1 1>}
  NamedRerouteDeclaration(None=Panner(Coordinate=@n153, Time=@n154) {speedX=-0.2, speedY=0.1}) {name=Panning Coordinates 2, nodeColor=<rgba 1 0.9141 0 1>}
  NamedRerouteDeclaration(None=AppendVector(A=@n85, B=Constant())) {name=Testing, nodeColor=<rgba 0 1 0.5313 1>}
  n155 = Constant3Vector() {constant=<rgba 1 0 0 0>}
  n156 = StaticSwitchParameter(
   True=MaterialFunctionCall() {materialFunction=&TangentBasis}.X U
   False=Normalize(VectorInput=Transform(None=@n155) {transformType=TRANSFORM_Local})
  ) {
   defaultValue=true
   desc="Saves 40 vertex instructions, but tangent to local is needed if the tornado mesh component rotates"
   parameterName=SkipTangentToLocalTransform
  }
  n157 = NamedRerouteDeclaration(None=ScalarParameter() {defaultValue=-1, parameterName=FlipRotation}) {name=FlipRotation, nodeColor=<rgba 1 0 0.4922 1>}
  n158 = NamedRerouteUsage()
  n159 = Constant3Vector() {constant=<rgba 0 0 -1 0>}
  n160 = ScalarParameter() {defaultValue=0.2, parameterName=CloseRangeMaxRotation}
  n161 = Subtract(A=Multiply(A=@n148) {constB=2})
  n162 = TextureCoordinate()
  n163 = Subtract(
   A=@n146
   B=LinearInterpolate(
    A=ScalarParameter() {defaultValue=10000, parameterName=CloseRangeRotationBlendStart}
    B=ScalarParameter() {defaultValue=20000, parameterName=CloseRangeRotationBlendStart-Top}
   ) {constAlpha=0}
  )
  n164 = ScalarParameter() {defaultValue=5000, parameterName=CloseRangeRotationBlendRange}
  n165 = LinearInterpolate(
   A=Multiply(A=DotProduct(A=@n158, B=@n159), B=Multiply(A=@n160, B=@n161))
   B=CurveAtlasRowParameter(CurveTime=ComponentMask(None=@n162) {r=true}) {defaultValue=1, parameterName=RotationCurve}.R
   Alpha=Saturate(None=Divide(A=@n163, B=@n164))
  )
  n166 = NamedRerouteUsage()
  n167 = NamedRerouteUsage()
  n168 = Add(A=NamedRerouteUsage(), B=NamedRerouteUsage())
  n169 = PreSkinnedPosition()
  NamedRerouteDeclaration(None=RotateAboutAxis(
    NormalizedRotationAxis=@n156
    RotationAngle=Multiply(A=@n157, B=@n165) {constB=-1}
    PivotPoint=Add(A=@n166, B=@n167)
    Position=Add(A=@n168, B=@n169)
   )) {name=RibbonRotationPosOffset, nodeColor=<rgba 1 0 0.7969 1>}
  n170 = NamedRerouteUsage()
  n171 = ComponentMask(None=@n170) {r=true}
  n172 = Add(A=@n170, B=AppendVector(A=Constant(), B=Multiply(A=@n171) {constB=-1}))
  n173 = Panner(Coordinate=@n172) {speedX=-0.1}
  n174 = ComponentMask(None=ConstantBiasScale(None=TextureSample!(UVs=@n173, Tex, Apply View MipBias) {samplerType=SAMPLERTYPE_LinearColor, texture=&T_Turbulence_AdditiveWarp_Linear}.RGB) {bias=-0.5, scale=2}) {g=true, r=true}
  n175 = SmoothStep(
   Min=Constant() {r=0.5}
   Max=Constant() {r=0.8}
   Value=MaterialFunctionCall(
    Volume Texture=TextureObject() {samplerType=SAMPLERTYPE_LinearColor, texture=&T_Lackluster_Streaks}
    UV=Panner(Coordinate=@n172) {speedX=-0.1, speedY=0.1}
    Velocity=Divide(A=@n174) {constB=3.5}
   ) {materialFunction=&Flowmaps_2D}.Result RGB
  )
  NamedRerouteDeclaration(None=Saturate(None=ComponentMask(None=@n175) {r=true})) {name=Swirls, nodeColor=<rgba 1 0 0.5859 1>}
  n176 = Subtract(A=ActorPositionWS(), B=WorldPosition().XYZ)
  n177 = Normalize(VectorInput=ComponentMask(None=@n176) {g=true, r=true})
  n178 = Divide(A=Arctangent2Fast(Y=ComponentMask(None=@n177) {r=true}, X=ComponentMask(None=@n177) {g=true})) {constB=6.28}
  n179 = Divide(A=ComponentMask(None=@n176) {b=true}) {constB=10092}
  NamedRerouteDeclaration(None=AppendVector(A=Add(A=@n178) {constB=0.5}, B=NamedRerouteDeclaration(None=@n179) {name=Height Gradient, nodeColor=<rgba 0.8047 1 0 1>})) {name=Cylindrical UVs, nodeColor=<rgba 1 0.4922 0 1>}
  n180 = ScalarParameter() {defaultValue=0.3, parameterName=Global Motion Speed}
  NamedRerouteDeclaration(None=Multiply(A=NamedRerouteDeclaration(None=@n180) {name=Global Motion Speed, nodeColor=<rgba 0 1 0.4844 1>}, B=Time())) {name=Global Motion Time, nodeColor=<rgba 0.875 0 1 1>}
  n181 = MaterialFunctionCall() {materialFunction=&TangentBasis}
  n182 = TextureCoordinate()
  n183 = Subtract(A=Multiply(A=ComponentMask(None=@n182) {g=true}) {constB=2})
  n184 = LinearInterpolate(
   A=ScalarParameter() {defaultValue=-0.1, parameterName=CurvedNormalBack}
   B=ScalarParameter() {defaultValue=0.04, parameterName=CurvedNormal}
   Alpha=@n143
  )
  n185 = Multiply(A=Add(A=@n165, B=Multiply(A=@n183, B=@n184)), B=NamedRerouteUsage())
  n186 = Constant()
  n187 = VertexNormalWS()
  n188 = Normalize(VectorInput=Add(A=RotateAboutAxis(NormalizedRotationAxis=@n181.X U, RotationAngle=@n185, PivotPoint=@n186, Position=@n187) {desc=wo}, B=@n187))
  NamedRerouteDeclaration(None=ShaderStageSwitch(PixelShader=VertexInterpolator(VS=@n188).PS, VertexShader=@n188)) {name=RibbonNormalRotated, nodeColor=<rgba 1 0 0.7969 1>}
  n189 = ComponentMask(None=NamedRerouteUsage()) {g=true, r=true}
  n190 = Constant()
  NamedRerouteDeclaration(None=Normalize(VectorInput=AppendVector(A=@n189, B=@n190))) {name=SplinePoint2DVector, nodeColor=<rgba 0.9453 1 0 1>}
  n191 = CurveAtlasRowParameter(CurveTime=ComponentMask(None=TextureCoordinate()) {r=true}) {defaultValue=3, parameterName=NoiseCurve}
  n192 = Multiply(A=Divide(A=NamedRerouteUsage(), B=Max(A=@n191.R) {constB=0.0001}), B=VectorParameter() {defaultValue=<rgba 1 1 1 0>}.RGB)
  n193 = Sine(None=Multiply(A=Time(), B=Constant() {r=0.25}))
  n194 = Add(
   A=LinearInterpolate(A=ComponentMask(None=@n192) {r=true}, B=ComponentMask(None=@n192) {g=true}, Alpha=ConstantBiasScale(None=@n193))
   B=ComponentMask(None=@n192) {b=true}
  )
  n195 = Frac(None=Multiply(A=Time(), B=ScalarParameter() {defaultValue=-0.2, group=Slurp, parameterName=Noise_Speed}) {constB=0.6})
  n196 = Add(A=Sine(None=Subtract(A=@n194, B=@n195)))
  n197 = Multiply(A=NamedRerouteUsage(), B=Multiply(A=Multiply(A=@n196) {constB=0.5}, B=@n191.G))
  n198 = NamedRerouteUsage()
  NamedRerouteDeclaration(None=Multiply(A=Multiply(A=@n197, B=@n198) {constB=0.09}, B=ScalarParameter() {defaultValue=0.8, parameterName=GlobalNoiseScale})) {name=SplinePointNoisePosOffset, nodeColor=<rgba 0 0.8828 1 1>}
  n199 = CurveAtlasRowParameter(CurveTime=ComponentMask(None=TextureCoordinate()) {r=true}) {defaultValue=2, parameterName=ProfileCurve}
  n200 = NamedRerouteUsage()
  NamedRerouteDeclaration(None=Multiply(A=NamedRerouteUsage(), B=Multiply(A=@n199.R, B=@n200) {constB=0.09})) {name=ProfilePosOffset, nodeColor=<rgba 0 1 0.5078 1>}
  NamedRerouteDeclaration(None=Constant() {r=0.09}) {name=ModelScale, nodeColor=<rgba 0.9453 1 0 1>}
  n201 = OneMinus(None=SphereMask(A=CameraPositionWS(), B=WorldPosition().XYZ, Radius=Constant() {r=300}, Hardness=Constant() {r=0.5}))
  n202 = MaterialFunctionCall() {materialFunction=&BoundingBoxBased_0-1_UVW}
  NamedRerouteDeclaration(None=Multiply(A=Multiply(A=@n45, B=@n201), B=SmoothStep(Value=@n202.B) {constMax=0.7, constMin=1.1})) {name=Opacity_Out, nodeColor=<rgba 1 0 0.75 1>}
  NamedRerouteUsage()
  NamedRerouteUsage()
  OneMinus(None=Fresnel(ExponentIn=ScalarParameter() {defaultValue=0.8, group=Color, parameterName=Fres_EX}) {baseReflectFraction=0, exponent=1})
  ScalarParameter() {defaultValue=0.33, group=Color, parameterName=DiffuseBrightness}
  ScalarParameter() {defaultValue=0.7, group=Color, parameterName=Roughness}
  ScalarParameter() {defaultValue=1, group=Color, parameterName=InscatterRoughness}
  SetMaterialAttributes!(MaterialAttributes, Emissive Color=NamedRerouteUsage()) {attributeSetTypes [#6]}
  TextureCoordinate()
  TextureSample(UVs=@n15) {samplerType=SAMPLERTYPE_LinearGrayscale, texture=/Game/Athena/Prototype/Terrain/T_CloudCoverage}
  VertexInterpolator(VS=@n120)
 ]
}
