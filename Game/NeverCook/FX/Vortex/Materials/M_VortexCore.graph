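# Two-sided lit material for the NeverCook vortex core FX, blending a sun-direction-tinted, emissive core layer with a glossy Lackluster-style chrome layer.
# A turbulence-warped flowmap of panning streaks forms the sharp blend mask, and a flowmapped swirl supplies the flattened normal.
# The chrome combines curve-atlas reflections, fresnel hue shift, animated phased and triangle nanomist emissive, and time-of-day scaling.
# A temporal-dither fade along V is built for the opacity mask override.
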Material /Game/NeverCook/FX/Vortex/Materials/M_VortexCore
graph {
 properties {bUseMaterialAttributes=true, pixelDepthOffsetMode=PDOM_Legacy, twoSided=true, usedShadingModels=Default Lit}
 nodes [
  n1 = Constant() {r=1}
  n2 = SetMaterialAttributes!(
   MaterialAttributes
   Base Color=NamedRerouteUsage()
   Emissive Color=NamedRerouteUsage()
   Metallic=@n1
   Specular=@n1
   Roughness=Constant() {r=0.1}
  ) {attributeSetTypes [
    #1
    #2
    #3
    #4
    #5
   ]}
  n3 = Constant() {r=0.6}
  n4 = TextureCoordinate()
  n5 = Add(A=@n4, B=AppendVector(A=Constant(), B=ComponentMask(None=@n4) {r=true}))
  n6 = Panner(Coordinate=@n5) {speedX=-0.1, speedY=0.1}
  n7 = TextureSample(UVs=Panner(Coordinate=@n5) {speedX=-0.1}) {samplerType=SAMPLERTYPE_LinearColor, texture=/Game/Effects/Fort_Effects/Textures/Plasma/T_Turbulence_AdditiveWarp_Linear}
  n8 = Divide(A=ComponentMask(None=ConstantBiasScale(None=@n7.RGB) {bias=-0.5, scale=2}) {g=true, r=true}) {constB=3.5}
  n9 = SmoothStep(
   Min=@n3
   Max=@n3
   Value=MaterialFunctionCall!(
    Volume Texture=TextureObject() {samplerType=SAMPLERTYPE_LinearColor, texture=/Game/TechArt/Seasonal/LackLuster/Textures/T_Lackluster_Streaks}
    UV=@n6
    Velocity=@n8
    Override Time
    Mip Level
   ) {materialFunction=/Engine/Functions/Engine_MaterialFunctions02/Texturing/Flowmaps_2D}.Result RGB
  )
  n10 = BlendMaterialAttributes(A=NamedRerouteUsage(), B=@n2, Alpha=Saturate(None=ComponentMask(None=@n9) {r=true}))
  n11 = MaterialFunctionCall(
   Volume Texture=TextureObject() {samplerType=SAMPLERTYPE_Normal, texture=/Game/TechArt/Seasonal/LackLuster/Textures/T_Lackluster_Swirl_N}
   UV=@n6
   Velocity=@n8
  ) {materialFunction=&Flowmaps_2D}
  n12 = ScalarParameter() {defaultValue=0.9, parameterName=Flatten Normal}
  MaterialFunctionCall!(
   Material=SetMaterialAttributes!(
    MaterialAttributes=MaterialFunctionCall!(Material=@n10, WorldPositionOffset) {materialFunction=/Engine/Functions/MaterialLayerFunctions/MatLayerBlend_AddWorldPositionOffset}.Blended Material
    Normal=MaterialFunctionCall(Normal=@n11.Result RGB, Flatness=@n12) {materialFunction=&FlattenNormal}.Result
   ) {attributeSetTypes [#6]}
   NewOpacityMask=NamedRerouteUsage()
   Mask
  ) {materialFunction=/Engine/Functions/MaterialLayerFunctions/MatLayerBlend_OverrideOpacityMask}
  n13 = TextureCoordinate()
  NamedRerouteDeclaration(None=DDX(Value=@n13)) {name=TexCoord_DDX, nodeColor=<rgba 0.7109 1 0 1>}
  NamedRerouteDeclaration(None=DDY(Value=@n13)) {name=TexCoord_DDY, nodeColor=<rgba 1 0 0.3281 1>}
  NamedRerouteDeclaration(None=Divide(A=TextureCoordinate(), B=ScalarParameter() {defaultValue=1, parameterName=Chrome_Divisor_UV_a})) {name=Chrome_UVs_A_NoPanning, nodeColor=<rgba 0 1 0.7656 1>}
  n14 = TextureCoordinate()
  n15 = ScalarParameter() {defaultValue=175, parameterName=NanomistTriangleUVDivisor}
  n16 = ComponentMask(None=PixelNormalWS()) {g=true, r=true}
  n17 = ScalarParameter() {defaultValue=0.05, parameterName=NanomistDistortionScale}
  n18 = TextureSample(UVs=Add(A=Divide(A=@n14, B=@n15) {constB=500}, B=Multiply(A=@n16, B=@n17))) {samplerType=SAMPLERTYPE_LinearColor, texture=/Game/TechArt/Seasonal/LackLuster/Textures/T_Lackluster_Triangles}
  n19 = Add(A=@n18.R, B=Multiply(A=Time(), B=ScalarParameter() {defaultValue=0.15, parameterName=NanomistSpeed}))
  n20 = NamedRerouteDeclaration(None=Multiply(A=Time(), B=ScalarParameter() {defaultValue=1, group=Chrome_Phased, parameterName=PhasePannerSpeedScale})) {name=PhasedPannerSpeedScale, nodeColor=<rgba 0.7745 0 1 1>}
  n21 = ReflectionVectorWS()
  n22 = Multiply(A=Abs(None=ComponentMask(None=@n21) {b=true, r=true}), B=ScalarParameter() {defaultValue=0.3, parameterName=PhaseReflectionVectorScale})
  n23 = TextureSampleParameter2D(UVs=Panner(Coordinate=@n22, Time=@n20) {bFractionalPart=true, speedX=0.0305, speedY=0.01}) {
   group=Chrome_Phased
   parameterName=PhasedMap
   samplerSource=SSM_Wrap_WorldGroupSettings
   texture=/Game/TechArt/Seasonal/LackLuster/Textures/T_Lackluster_Phased
  }
  n24 = Constant2Vector() {g=-1, r=1}
  n25 = TextureSampleParameter2D(UVs=Panner(Coordinate=Multiply(A=@n22, B=@n24) {constB=-1}, Time=@n20) {bFractionalPart=true, speedX=0.0225, speedY=0.0065}) {
   group=Chrome_Phased
   parameterName=PhasedMap
   samplerSource=SSM_Wrap_WorldGroupSettings
   texture=&T_Lackluster_Phased
  }
  n26 = Multiply(A=@n23.RGB, B=@n25.RGB)
  n27 = Max(A=@n23.RGB, B=@n25.RGB)
  n28 = ScalarParameter() {defaultValue=0.5, parameterName=PhasedMapLerp}
  n29 = ScalarParameter() {defaultValue=0.09, parameterName=UnphasedColorScale}
  n30 = ScalarParameter() {defaultValue=0.25, parameterName=PhasedColorScale}
  n31 = Constant() {r=1}
  n32 = Multiply(
   A=ConstantBiasScale(None=Sine(None=@n19) {period=0.5})
   B=Multiply(A=LinearInterpolate(A=@n26, B=@n27, Alpha=@n28) {constAlpha=0.25}, B=LinearInterpolate(A=@n29, B=@n30, Alpha=@n31))
  )
  n33 = TextureCoordinate() {uTiling=3, vTiling=3}
  n34 = AppendVector(A=Constant(), B=ComponentMask(None=@n33) {r=true})
  n35 = TextureSample() {samplerType=SAMPLERTYPE_LinearColor, texture=&T_Turbulence_AdditiveWarp_Linear}
  n36 = Add(A=Panner(Coordinate=Add(A=@n33, B=@n34)) {speedX=-0.3}, B=Multiply(A=ComponentMask(None=@n35.RGB) {g=true, r=true}) {constB=0.2})
  n37 = NamedRerouteUsage()
  n38 = NamedRerouteUsage()
  n39 = TextureSampleParameter2D!(UVs=Panner(Coordinate=@n36) {speedX=0.05, speedY=0.02}, "DDX(UVs)"=@n37, "DDY(UVs)"=@n38, Apply View MipBias) {
   group=Chrome_Phased
   mipValueMode=TMVM_Derivative
   parameterName=PhasedMap
   texture=&T_Lackluster_Phased
  }
  n40 = TextureSampleParameter2D(UVs=Panner(Coordinate=@n36) {speedX=0.02, speedY=0.05}, "DDX(UVs)"=@n37, "DDY(UVs)"=@n38) {
   group=Chrome_Phased
   mipValueMode=TMVM_Derivative
   parameterName=PhasedMap
   texture=&T_Lackluster_Phased
  }
  n41 = Add(A=Multiply(B=@n39.RGB) {constA=1}, B=@n40.RGB)
  n42 = Multiply(A=ScalarParameter() {defaultValue=1.5, parameterName=PhasedLineScale}, B=Multiply(A=OneMinus(None=@n18.G), B=@n41))
  n43 = Multiply(
   A=Desaturation(None=@n41, Fraction=ScalarParameter() {defaultValue=0.3, parameterName=PhaseSaturation})
   B=MaterialFunctionCall(
    Value_Morning=ScalarParameter() {defaultValue=0.15, parameterName=PhaseColorScale_Morning}
    Value_Day=ScalarParameter() {defaultValue=0.4, parameterName=PhaseColorScale_Day}
    Value_Evening=ScalarParameter() {defaultValue=0.2, parameterName=PhaseColorScale_Evening}
    Value_Night=ScalarParameter() {defaultValue=0.275, parameterName=PhaseColorScale_Night}
   ) {materialFunction=&Scalar_Time_of_Day}.Result
  )
  NamedRerouteDeclaration(None=Add(A=Multiply(A=@n18.G, B=@n32), B=Add(A=@n42, B=@n43))) {name=Emissive, nodeColor=<rgba 1 0 0.8203 1>}
  n44 = ComponentMask(None=PixelNormalWS()) {g=true, r=true}
  n45 = ScalarParameter() {defaultValue=1, group=ReflectionCurve, parameterName=ReflectionCurveUVDistortionScale}
  n46 = TextureSample(UVs=Add(A=Multiply(A=@n44, B=@n45), B=NamedRerouteUsage())) {samplerType=SAMPLERTYPE_LinearColor, texture=&T_Lackluster_Streaks}
  n47 = MaterialFunctionCall(Exponent=ScalarParameter() {defaultValue=1.5, group=HueShift, parameterName=FresnelHueShiftExponent}) {materialFunction=&MF_Fresnel_FX}
  n48 = CurveAtlasRowParameter(CurveTime=Frac(None=Add(A=@n46.R, B=@n47.Fresnel))) {
   atlas=/Game/TechArt/Seasonal/LackLuster/Curves/CA_Lackluster
   curve=/Game/TechArt/Seasonal/LackLuster/Curves/C_Lackluster_Reflection
   parameterName=ReflectionCurve
  }
  n49 = ScalarParameter() {defaultValue=1, parameterName=ReflectionCurveScale}
  n50 = MaterialFunctionCall(
   Value_Morning=VectorParameter() {group=BaseColor, parameterName=BaseColor_Morning, defaultValue=<rgba 0.4323 0.2592 0.384 0>}.RGB
   Value_Day=VectorParameter() {group=BaseColor, parameterName=BaseColor_Day, defaultValue=<rgba 0.8073 0.7711 0.5256 0>}.RGB
   Value_Evening=VectorParameter() {group=BaseColor, parameterName=BaseColor_Evening, defaultValue=<rgba 0.4358 0.4257 0.3578 0>}.RGB
   Value_Night=VectorParameter() {group=BaseColor, parameterName=BaseColor_Night, defaultValue=<rgba 0.7951 0.7769 0.653 0>}.RGB
  ) {materialFunction=&Scalar_Time_of_Day}
  n51 = LinearInterpolate(A=@n50.Result, B=Multiply(A=@n50.Result, B=ScalarParameter() {defaultValue=1.5, parameterName=BaseColorScalePhased}), Alpha=Constant() {r=1})
  n52 = Frac(None=Multiply(A=@n47.Fresnel, B=ScalarParameter() {defaultValue=1.5, group=HueShift, parameterName=HueShiftScale}))
  n53 = VectorParameter() {group=HueShift, parameterName=HueShiftColor, defaultValue=<rgba 0.75 0.8642 1 0>}
  n54 = Add(
   A=Multiply(A=Multiply(A=@n48, B=@n49), B=@n47.Fresnel Inverted)
   B=Multiply(
    A=StaticSwitchParameter(True=@n51, False=@n50.Result) {defaultValue=true, parameterName=IsPhased}
    B=MaterialFunctionCall(Hue Shift Percentage=@n52, Texture=@n53.RGB) {materialFunction=&HueShift}.Result
   )
  )
  NamedRerouteDeclaration(None=Desaturation(None=Add(A=@n54) {constB=0}, Fraction=ScalarParameter() {defaultValue=-0.5, parameterName=BaseColorDesaturationFraction})) {name=Bushki, nodeColor=<rgba 0 0.6484 1 1>}
  n55 = ScalarParameter() {defaultValue=0.2, parameterName=Fade}
  n56 = ComponentMask(None=TextureCoordinate()) {g=true}
  NamedRerouteDeclaration(None=MaterialFunctionCall(Alpha Threshold=SmoothStep(Max=@n55, Value=@n56)) {materialFunction=&DitherTemporalAA}.Result) {name=Fade, nodeColor=<rgba 0.3828 0 1 1>}
  n57 = MaterialFunctionCall() {materialFunction=/Game/Global/MaterialFunctions/FNCollectionParameters/MF_SunParameters_FNC}
  n58 = DotProduct(A=@n57.Direction, B=ReflectionVectorWS())
  n59 = Multiply(A=DotProduct(A=@n57.Direction, B=PixelNormalWS())) {constB=-1}
  n60 = LinearInterpolate(
   A=LinearInterpolate(
    A=VectorParameter() {parameterName=Core Secondary Color, defaultValue=<rgba 0.0953 0.2542 0.7529 1>}.RGB
    B=VectorParameter() {parameterName=Core Primary Color, defaultValue=<rgba 0.2502 0.9216 0.9823 1>}.RGB
    Alpha=Saturate(None=@n58)
   )
   B=VectorParameter() {parameterName=Scatter Color, defaultValue=<rgba 0.3191 0.225 0.5469 1>}.RGB
   Alpha=NamedRerouteDeclaration(None=Saturate(None=@n59)) {name=Dotto, nodeColor=<rgba 0 1 0.8359 1>}
  )
  n61 = SetMaterialAttributes!(
   MaterialAttributes
   Base Color=@n60
   Normal=Constant3Vector() {constant=<rgba 0 0 1 0>}
   Emissive Color=Multiply(A=@n60, B=ScalarParameter() {defaultValue=0.5, parameterName=Emissive Multiplier})
   Metallic=Constant() {r=1}
   Roughness=Constant() {r=1}
  ) {attributeSetTypes [
    #1
    #6
    #2
    #3
    #5
   ]}
  NamedRerouteDeclaration(None=@n61) {name=Vortex Core, nodeColor=<rgba 0 1 0.2969 1>}
 ]
}
