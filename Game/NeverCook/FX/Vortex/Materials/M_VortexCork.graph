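# Two-sided vortex cork effect material that blends a chrome layer with a metallic, emissive vortex core.
# Core color combines a panning swirl with a sun-driven curve-atlas gradient.
# A phase mask comes from flowmapped noise, alongside actor-relative cylindrical UVs and a dithered fade along U.
# Layer blends add world position offset, built from sine/cosine rotation, normal-aligned ripples and scrolling noise displacement, and override the opacity mask.
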Material /Game/NeverCook/FX/Vortex/Materials/M_VortexCork
graph {
 properties {
  bUseMaterialAttributes=true
  editorOnlyData="<self>:M_VortexCoreEditorOnlyData"
  pixelDepthOffsetMode=PDOM_Legacy
  previewMesh=/Game/NeverCook/FX/Vortex/Meshes/SM_Vortex_Cork
  twoSided=true
  usedShadingModels=Default Lit
 }
 nodes [
  n1 = BlendMaterialAttributes(A=NamedRerouteUsage(), B=NamedRerouteUsage(), Alpha=NamedRerouteUsage())
  n2 = NamedRerouteUsage()
  MaterialFunctionCall!(
   Material=SetMaterialAttributes!(
    MaterialAttributes=MaterialFunctionCall!(Material=@n1, WorldPositionOffset=@n2) {materialFunction=/Engine/Functions/MaterialLayerFunctions/MatLayerBlend_AddWorldPositionOffset}.Blended Material
    Normal
   ) {attributeSetTypes [#1]}
   NewOpacityMask=NamedRerouteUsage()
   Mask
  ) {materialFunction=/Engine/Functions/MaterialLayerFunctions/MatLayerBlend_OverrideOpacityMask}
  NamedRerouteDeclaration(None=MaterialFunctionCall!(BlendMaterialAttributesAlpha, MA_In, WorldUVsOverride, AlbedoOverride, SourceNormalsStrengthOverride, bBlendMaterialAttributes) {materialFunction=/Game/NeverCook/TechArt/LackLuster/LookDev/Materials/MF_Chrome_Base}.MA Out) {name=Chrome, nodeColor=<rgba 0 0.5547 1 1>}
  n3 = NamedRerouteUsage()
  n4 = MaterialFunctionCall(Y=Multiply(A=ComponentMask(None=@n3) {r=true}) {constB=-1}) {materialFunction=&MakeFloat2}
  n5 = TextureObject() {samplerType=SAMPLERTYPE_LinearColor, texture=/Game/Effects/Fort_Effects/Textures/Plasma/T_Turbulence_AdditiveWarp_Linear}
  n6 = TextureSample(UVs=Panner(Coordinate=Subtract(A=@n3, B=@n4.Result)) {speedX=-0.05, speedY=0.02}, Tex=@n5) {samplerType=SAMPLERTYPE_LinearColor, texture=&T_Turbulence_AdditiveWarp_Linear}
  n7 = Time()
  n8 = Time()
  n9 = DotProduct(
   A=MaterialFunctionCall!(
    Volume Texture=TextureObject() {samplerType=SAMPLERTYPE_LinearColor, texture=&HighContrastNoise}
    UV=NamedRerouteUsage()
    Velocity=ComponentMask(None=@n6.RGB) {g=true, r=true}
    Override Time=Divide(A=@n7) {constB=10000}
    Mip Level
   ) {materialFunction=/Engine/Functions/Engine_MaterialFunctions02/Texturing/Flowmaps_2D}.Result RGB
   B=MaterialFunctionCall(Hue Shift Percentage=Divide(A=@n8) {constB=20000}) {materialFunction=&HueShift}.Result
  )
  NamedRerouteDeclaration(None=SmoothStep(Value=Multiply(A=@n9)) {constMax=0.52, constMin=0.5}) {name=Phase Mask, nodeColor=<rgba 0.9688 0 1 1>}
  n10 = Panner() {speedY=0.2}
  NamedRerouteDeclaration(None=ConstantBiasScale(None=TextureSample(UVs=@n10, Tex=@n5).RGB) {bias=-0.5, scale=2}) {name=Displacement Noise, nodeColor=<rgba 0.4062 1 0 1>}
  n11 = Constant() {r=1}
  n12 = ScalarParameter() {defaultValue=0.75, parameterName=Fade}
  n13 = ComponentMask(None=TextureCoordinate()) {r=true}
  NamedRerouteDeclaration(None=MaterialFunctionCall(Alpha Threshold=SmoothStep(Min=@n11, Max=@n12, Value=@n13)) {materialFunction=&DitherTemporalAA}.Result) {name=Fade, nodeColor=<rgba 0.3828 0 1 1>}
  n14 = Subtract(A=ActorPositionWS(), B=WorldPosition() {worldPositionShaderOffset=WPT_ExcludeAllShaderOffsets}.XYZ)
  n15 = Normalize(VectorInput=ComponentMask(None=@n14) {g=true, r=true})
  n16 = Divide(A=Arctangent2Fast(Y=ComponentMask(None=@n15) {r=true}, X=ComponentMask(None=@n15) {g=true})) {constB=6.28}
  n17 = Divide(A=ComponentMask(None=@n14) {b=true}) {constB=10092}
  NamedRerouteDeclaration(None=AppendVector(A=Add(A=@n16) {constB=0.5}, B=NamedRerouteDeclaration(None=@n17) {name=Height Gradient, nodeColor=<rgba 0.8047 1 0 1>})) {name=Cylindrical UVs, nodeColor=<rgba 1 0.4922 0 1>}
  n18 = TextureCoordinate()
  n19 = AppendVector(A=ScalarParameter() {defaultValue=10, parameterName=Pattern Tiling}, B=Constant() {r=1})
  n20 = NamedRerouteDeclaration(None=Panner(Coordinate=Multiply(A=@n18, B=@n19)) {speedX=-0.5, speedY=-0.3}) {name=Swirl Panner, nodeColor=<rgba 0 1 0.5547 1>}
  n21 = TextureObject() {texture=null}
  n22 = MaterialFunctionCall() {materialFunction=/Game/Global/MaterialFunctions/FNCollectionParameters/MF_SunParameters_FNC}
  n23 = ReflectionVectorWS()
  n24 = Multiply(A=Saturate(None=DotProduct(A=@n22.Direction, B=@n23))) {constB=0.5}
  n25 = PixelNormalWS()
  n26 = Saturate(None=Multiply(A=DotProduct(A=@n22.Direction, B=@n25)) {constB=-1})
  n27 = CurveAtlasRowParameter(CurveTime=Add(A=Subtract(B=@n24) {constA=0.5}, B=Multiply(A=@n26) {constB=0.5})) {
   atlas=&/Atlas_NanoVortex
   curve=&/Curve_NanoVortex
   parameterName=Color Curve
  }
  n28 = Multiply(A=LinearInterpolate(Alpha=TextureSample(UVs=@n20, Tex=@n21).R) {constA=0.8}, B=@n27)
  n29 = SetMaterialAttributes!(
   MaterialAttributes
   Base Color=@n28
   Normal=Constant3Vector() {constant=<rgba 0 0 1 0>}
   Emissive Color=Multiply(A=@n28, B=ScalarParameter() {defaultValue=0.5, parameterName=Emissive Multiplier})
   Metallic=Constant() {r=1}
   Roughness=Constant()
  ) {attributeSetTypes [
    #2
    #1
    #3
    #4
    #5
   ]}
  NamedRerouteDeclaration(None=@n29) {name=Vortex Core, nodeColor=<rgba 0 1 0.2969 1>}
  n30 = Time()
  n31 = Divide(A=Add(A=Divide(A=@n30) {constB=4}, B=NamedRerouteUsage()), B=ScalarParameter() {defaultValue=10, parameterName=Rotation Speed})
  n32 = Sine(None=@n31)
  n33 = Cosine(None=Add(A=@n31))
  n34 = NamedRerouteUsage()
  n35 = ScalarParameter() {defaultValue=2000, parameterName=Param_4}
  n36 = Add(B=Multiply(
    A=MaterialFunctionCall(X=@n32, Y=@n33) {materialFunction=&MakeFloat3}.Result
    B=Multiply(A=@n34, B=@n35)
   ) {constB=300})
  n37 = Time()
  n38 = Add(A=Add(A=Divide(A=@n37) {constB=20}, B=VertexColor().R), B=ComponentMask(None=TextureCoordinate()) {r=true})
  n39 = Multiply(A=Multiply(A=VertexNormalWS(), B=Sine(None=@n38) {period=0.05})) {constB=50}
  n40 = Multiply(A=NamedRerouteUsage(), B=VertexNormalWS())
  n41 = ScalarParameter() {defaultValue=100, parameterName=Dispacement}
  NamedRerouteDeclaration(None=Add(A=Add(A=@n36, B=@n39), B=Multiply(A=@n40, B=@n41))) {name=WPO, nodeColor=<rgba 1 0.3281 0 1>}
  VertexInterpolator(VS=PreSkinnedPosition())
 ]
}
