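# M_VortexEdge is a translucent unlit material for the vortex effect's edge, holding two material-attribute branches.
# One sets opacity from inverse Fresnel and offsets vertices with panned sine ripples along normals plus a time-rotating sway, phase-shifted by foam texture and scaled by radial distance.
# The other feeds flowmap-driven albedo and world-aligned normals into MF_Chrome_Base, adding vertex displacement from panning noise.
# The flowmap's texture slot is empty.
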Material /Game/NeverCook/FX/Vortex/Materials/M_VortexEdge
graph {
 properties {
  bUseMaterialAttributes=true
  blendMode=BLEND_Translucent
  editorOnlyData="<self>:M_VortexCoreEditorOnlyData"
  pixelDepthOffsetMode=PDOM_Legacy
  shadingModel=MSM_Unlit
  usedShadingModels=Unlit
 }
 nodes [
  n1 = Constant() {r=1}
  n2 = ScalarParameter() {defaultValue=20, parameterName=Param_2}
  n3 = Multiply(A=TextureCoordinate(), B=Multiply(A=AppendVector(A=@n1, B=@n2), B=ScalarParameter() {defaultValue=0.1, parameterName=Param_3}))
  n4 = ComponentMask(None=TextureCoordinate()) {r=true}
  n5 = ScalarParameter() {defaultValue=195.499}
  n6 = LinearInterpolate(Alpha=Power(Base=Divide(A=@n4, B=@n5), Exp=ScalarParameter() {defaultValue=0.1, parameterName=Param_1}) {constExponent=0.2}) {constA=1, constB=0}
  n7 = Constant() {r=0.1}
  n8 = ComponentMask(None=MaterialFunctionCall!(
    Volume Texture=TextureObject() {texture=null}
    UV=@n3
    Velocity=MaterialFunctionCall(X=@n6, Y=@n7) {materialFunction=&MakeFloat2}.Result
    Override Time
    Mip Level
   ) {materialFunction=/Engine/Functions/Engine_MaterialFunctions02/Texturing/Flowmaps_2D}.Result RGB) {r=true}
  n9 = MaterialFunctionCall(
   TextureObject=TextureObject() {samplerType=SAMPLERTYPE_Normal, texture=/Game/Global/Textures/Default/Tileable_Patterns/1Normal}
   TextureSize=Constant3Vector() {constant=<rgba 1000 1000 10000 0>}
  ) {materialFunction=&WorldAlignedNormal}
  n10 = Panner(Coordinate=@n3) {speedX=-0.5}
  n11 = TextureSample(UVs=@n10) {samplerType=SAMPLERTYPE_Normal, texture=/Game/Global/Textures/Default/Noise/T_Normal_Noise01}
  n12 = VertexNormalWS()
  MaterialFunctionCall!(
   Material=MaterialFunctionCall!(
    BlendMaterialAttributesAlpha
    MA_In
    WorldUVsOverride
    AlbedoOverride=LinearInterpolate(Alpha=@n8) {constA=0.75}
    SourceNormalsStrengthOverride=@n9.XYZ Texture
    bBlendMaterialAttributes
   ) {materialFunction=/Game/NeverCook/TechArt/LackLuster/LookDev/Materials/MF_Chrome_Base}.MA Out
   WorldPositionOffset=Multiply(A=Multiply(A=@n11.RGB, B=@n12), B=ScalarParameter() {defaultValue=20, parameterName=Param_4})
  ) {materialFunction=/Engine/Functions/MaterialLayerFunctions/MatLayerBlend_AddWorldPositionOffset}
  MaterialFunctionCall(BaseNormal=@n9.XYZ Texture, AdditionalNormal=@n11.RGB) {materialFunction=&BlendAngleCorrectedNormals}
  n13 = Fresnel() {baseReflectFraction=0, exponent=1}
  n14 = Panner(
   Coordinate=TextureCoordinate()
   Speed=AppendVector(A=ScalarParameter() {defaultValue=1, parameterName=Speed X}, B=ScalarParameter() {parameterName=Speed Y})
  )
  n15 = Multiply(A=ComponentMask(None=@n14) {g=true}, B=ScalarParameter() {defaultValue=10, parameterName=Frequency})
  n16 = Multiply(A=ComponentMask(None=@n14) {r=true})
  n17 = Multiply(A=Sine(None=Add(A=@n15, B=@n16)), B=VertexNormalWS())
  n18 = ScalarParameter() {defaultValue=30, parameterName=Distance}
  n19 = Distance(A=ComponentMask(None=PreSkinnedPosition()) {g=true, r=true}, B=Constant())
  n20 = Multiply(A=Power(Base=Divide(A=@n19) {constB=6760})) {constB=6760}
  n21 = Constant()
  n22 = Divide(A=ComponentMask(None=TextureCoordinate()) {g=true}, B=ScalarParameter() {defaultValue=10, parameterName=Param_5})
  n23 = Add(
   A=Divide(A=Time(), B=ScalarParameter() {defaultValue=10, parameterName=Rotation Speed})
   B=TextureSample(UVs=AppendVector(A=@n21, B=@n22)) {samplerType=SAMPLERTYPE_LinearColor, texture=/FortWater/Textures/02_WaterFoam_Veroni_base_H}.R
  )
  n24 = MaterialFunctionCall(X=Sine(None=@n23), Y=Cosine(None=@n23)) {materialFunction=&MakeFloat3}
  MaterialFunctionCall(
   Material=SetMaterialAttributes!(MaterialAttributes, Emissive Color, Opacity=OneMinus(None=@n13)) {attributeSetTypes [#1, #2]}
   WorldPositionOffset=Add(A=Multiply(A=@n17, B=@n18), B=Multiply(A=@n20, B=@n24.Result))
  ) {materialFunction=&MatLayerBlend_AddWorldPositionOffset}
  TextureSample(UVs=@n10)
 ]
}
