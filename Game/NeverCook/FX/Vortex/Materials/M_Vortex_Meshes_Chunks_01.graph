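# Lit material for vortex chunk meshes, with math that swirls, drifts and cycles chunks along a direction, vertical by default.
# A time- and height-driven RotateAboutAxis spiral, sized by UV1.G, adds to layered-sine drift scaled by Curl.
# A looping Frac phase, offset by UV1.G with vertex red blending its period, sweeps travel scaled to object radius.
# That cycle also feeds a sine mask, alongside a darkened flat color.
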Material /Game/NeverCook/FX/Vortex/Materials/M_Vortex_Meshes_Chunks_01
graph {
 properties {editorOnlyData="<self>:M_Vortex_Sprite_Chunks_01EditorOnlyData", pixelDepthOffsetMode=PDOM_Legacy, usedShadingModels=Default Lit}
 nodes [
  n1 = ComponentMask(None=WorldPosition().XYZ) {b=true}
  n2 = ScalarParameter() {defaultValue=15000, parameterName=SpiralFrequency}
  n3 = Time()
  n4 = ScalarParameter() {defaultValue=-0.025, parameterName=SpiralSpeed}
  n5 = Constant3Vector() {constant=<rgba 1 0 0 0>}
  n6 = ScalarParameter() {defaultValue=5000, parameterName=SpiralSize}
  n7 = Constant3Vector() {constant=<rgba 1 0 0 0>}
  n8 = ScalarParameter() {defaultValue=500, parameterName=SpiralSizeLow}
  n9 = TextureCoordinate() {coordinateIndex=1}
  n10 = RotateAboutAxis(
   NormalizedRotationAxis=Constant3Vector() {constant=<rgba 0 0 1 0>}
   RotationAngle=Add(A=Divide(A=@n1, B=@n2), B=Multiply(A=@n3, B=@n4))
   PivotPoint=Constant3Vector()
   Position=LinearInterpolate(A=Multiply(A=@n5, B=@n6), B=Multiply(A=@n7, B=@n8), Alpha=ComponentMask(None=@n9) {g=true})
  )
  n11 = Multiply(A=Time(), B=ScalarParameter() {defaultValue=0.001, parameterName=VarianceSpeed}) {constB=0.05}
  n12 = ComponentMask(None=TextureCoordinate() {coordinateIndex=1}) {g=true}
  n13 = Add(A=Multiply(A=Time(), B=ScalarParameter() {defaultValue=0.01, parameterName=DriftSpeed}), B=Sine(None=Add(A=@n11, B=@n12)) {period=0.17})
  n14 = AppendVector(A=AppendVector(A=Sine(None=@n13) {period=0.56}, B=Sine(None=@n13) {period=0.7}), B=Sine(None=@n13) {period=0.82})
  n15 = Add(A=@n11, B=@n12)
  n16 = Multiply(
   A=LinearInterpolate(
    A=ScalarParameter() {defaultValue=-100, parameterName=MinHorizontalDrift}
    B=ScalarParameter() {defaultValue=100, parameterName=MaxHorizontalDrift}
    Alpha=Sine(None=@n15)
   )
   B=Clamp(None=ScalarParameter() {defaultValue=1, parameterName=Curl}) {maxDefault=10}
  )
  n17 = Multiply(A=ObjectRadius(), B=ScalarParameter() {defaultValue=0.75, parameterName=MinVerticalSpeedMult}) {constB=0.5}
  n18 = Multiply(A=@n17) {constB=-1}
  n19 = Time()
  n20 = LinearInterpolate(
   A=ScalarParameter() {defaultValue=30, parameterName=VerticalSpeedLow}
   B=ScalarParameter() {defaultValue=30, parameterName=VerticalSpeed}
   Alpha=VertexColor().R
  )
  n21 = ComponentMask(None=TextureCoordinate() {coordinateIndex=1}) {g=true}
  n22 = Add(A=Frac(None=Divide(A=@n19, B=@n20)), B=Multiply(A=Subtract(A=@n21) {constB=0.5}) {constB=-2})
  n23 = Frac(None=@n22)
  Add(
   A=Add(A=Transform(None=@n10) {transformSourceType=TRANSFORMSOURCE_World, transformType=TRANSFORM_Local}, B=Multiply(A=@n14, B=@n16))
   B=Multiply(A=LinearInterpolate(A=@n18, B=@n17, Alpha=@n23), B=VectorParameter() {parameterName=OverallDirection, defaultValue=<rgba 0 0 1 1>}.RGB)
  )
  n24 = ConstantBiasScale(None=Sine(None=Add(A=@n22) {constB=-0.25}))
  Multiply(A=Constant() {r=1}, B=Power(Base=LinearInterpolate(Alpha=@n24)) {constExponent=1})
  Multiply(
   A=VectorParameter() {parameterName=Color, defaultValue=<rgba 0.25 0.25 0.25 1>}.RGB
   B=ScalarParameter() {defaultValue=0.5, parameterName=Darkness}
  )
  VertexNormalWS()
 ]
}
