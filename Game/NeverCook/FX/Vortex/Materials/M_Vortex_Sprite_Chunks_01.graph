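# Masked lit material for camera-facing rock-chunk sprites swirling in a vortex effect.
# World position offset sums sprite billboarding with distance-based sizing, a height-driven spiral around Z, sine drift and looping vertical travel, varied by UV1 and vertex color.
# A 4x4 rock atlas spins within each cell, with alpha clipped by a sphere mask and faded over the vertical cycle.
# Older helix-motion branches appear unused alongside the main offset.
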Material /Game/NeverCook/FX/Vortex/Materials/M_Vortex_Sprite_Chunks_01
graph {
 properties {blendMode=BLEND_Masked, pixelDepthOffsetMode=PDOM_Legacy, usedShadingModels=Default Lit}
 nodes [
  n1 = WorldPosition()
  n2 = TextureCoordinate() {coordinateIndex=1}
  n3 = Multiply(A=Subtract(A=ComponentMask(None=@n2) {g=true}) {constB=0.5}) {constB=-2}
  n4 = Multiply(A=Time(), B=ScalarParameter() {defaultValue=-0.01, parameterName=Slow})
  n5 = Fmod(
   A=Multiply(A=Add(A=@n3, B=@n4), B=VectorParameter() {parameterName=Direction, defaultValue=<rgba 0 0 3 1>}.RGB)
   B=ScalarParameter() {defaultValue=1, parameterName=HeightRepeat}
  )
  n6 = Subtract(A=ObjectPositionWS(), B=Add(A=@n1.XYZ, B=@n5))
  n7 = ComponentMask(None=Divide(A=Add(A=@n6, B=@n1.XYZ), B=ScalarParameter() {defaultValue=1, parameterName=Frequency}) {constB=100}) {b=true}
  n8 = Sine(None=@n7)
  n9 = Cosine(None=@n7)
  n10 = ScalarParameter() {defaultValue=5000, parameterName=Radius}
  n11 = ScalarParameter() {defaultValue=2500, parameterName=RadiusLow}
  n12 = ComponentMask(None=TextureCoordinate() {coordinateIndex=1}) {g=true}
  Add(
   A=Multiply(
    A=MaterialFunctionCall(R=@n8, G=@n9) {materialFunction=&AppendMany}.RGB
    B=LinearInterpolate(A=@n10, B=@n11, Alpha=@n12)
   )
   B=@n5
  )
  n13 = LinearInterpolate(
   A=ComponentMask(None=VectorParameter() {parameterName=SpriteSizeNear, defaultValue=<rgba 250 250 500 500>}.RGB) {g=true, r=true}
   B=ComponentMask(None=VectorParameter() {parameterName=SpriteSize, defaultValue=<rgba 1500 1500 600 500>}.RGB) {g=true, r=true}
   Alpha=MaterialFunctionCall(
    Fade_Start_Distance=ScalarParameter() {defaultValue=50000, parameterName="Fade_Start_Distance (S)"}
    Fade_End_Distance=ScalarParameter() {defaultValue=5000, parameterName="Fade_End_Distance (S)"}
   ) {materialFunction=&MF_Distance_SyncedFade_FX}.Result
  )
  n14 = LinearInterpolate(Alpha=VertexColor().G) {constA=0.1}
  n15 = Divide(A=ComponentMask(None=WorldPosition().XYZ) {b=true}, B=ScalarParameter() {defaultValue=15000, parameterName=SpiralFrequency})
  n16 = Multiply(A=Time(), B=ScalarParameter() {defaultValue=-0.025, parameterName=SpiralSpeed})
  n17 = Multiply(A=Constant3Vector() {constant=<rgba 1 0 0 0>}, B=ScalarParameter() {defaultValue=5000, parameterName=SpiralSize})
  n18 = Multiply(A=Constant3Vector() {constant=<rgba 1 0 0 0>}, B=ScalarParameter() {defaultValue=500, parameterName=SpiralSizeLow})
  n19 = ComponentMask(None=TextureCoordinate() {coordinateIndex=1}) {g=true}
  n20 = Transform(None=RotateAboutAxis(
    NormalizedRotationAxis=Constant3Vector() {constant=<rgba 0 0 1 0>}
    RotationAngle=Add(A=@n15, B=@n16)
    PivotPoint=Constant3Vector()
    Position=LinearInterpolate(A=@n17, B=@n18, Alpha=@n19)
   )) {transformSourceType=TRANSFORMSOURCE_World, transformType=TRANSFORM_Local}
  n21 = Multiply(A=Time(), B=ScalarParameter() {defaultValue=0.001, parameterName=VarianceSpeed}) {constB=0.05}
  n22 = ComponentMask(None=TextureCoordinate() {coordinateIndex=1}) {g=true}
  n23 = Add(A=Multiply(A=Time(), B=ScalarParameter() {defaultValue=0.01, parameterName=DriftSpeed}), B=Sine(None=Add(A=@n21, B=@n22)) {period=0.17})
  n24 = Sine(None=@n23) {period=0.56}
  n25 = Sine(None=@n23) {period=0.7}
  n26 = ScalarParameter() {defaultValue=-100, parameterName=MinHorizontalDrift}
  n27 = ScalarParameter() {defaultValue=100, parameterName=MaxHorizontalDrift}
  n28 = Sine(None=Add(A=@n21, B=@n22))
  n29 = ScalarParameter() {defaultValue=1, parameterName=Curl}
  n30 = Multiply(
   A=AppendVector(A=AppendVector(A=@n24, B=@n25), B=Sine(None=@n23) {period=0.82})
   B=Multiply(A=LinearInterpolate(A=@n26, B=@n27, Alpha=@n28), B=Clamp(None=@n29) {maxDefault=10})
  )
  n31 = Multiply(A=ObjectRadius(), B=ScalarParameter() {defaultValue=0.75, parameterName=MinVerticalSpeedMult}) {constB=0.5}
  n32 = Time()
  n33 = LinearInterpolate(
   A=ScalarParameter() {defaultValue=30, parameterName=VerticalSpeedLow}
   B=ScalarParameter() {defaultValue=30, parameterName=VerticalSpeed}
   Alpha=VertexColor().R
  )
  n34 = ComponentMask(None=TextureCoordinate() {coordinateIndex=1}) {g=true}
  n35 = Add(A=Frac(None=Divide(A=@n32, B=@n33)), B=Multiply(A=Subtract(A=@n34) {constB=0.5}) {constB=-2})
  n36 = LinearInterpolate(A=Multiply(A=@n31) {constB=-1}, B=@n31, Alpha=Frac(None=@n35))
  n37 = VectorParameter() {parameterName=OverallDirection, defaultValue=<rgba 0 0 1 1>}
  Add(
   A=MaterialFunctionCall!(Center Location=WorldPosition().XYZ, 0-1 UVs, XY scale=Multiply(A=@n13, B=@n14), OverridePivot, "Normals (Optional)") {materialFunction=&Sprite}.World position offset
   B=Add(A=Add(A=@n20, B=@n30), B=Multiply(A=@n36, B=@n37.RGB))
  )
  n38 = WorldPosition()
  n39 = Fmod(A=Multiply(A=Time(), B=Constant3Vector() {constant=<rgba 0 0 100 0>}), B=ScalarParameter() {defaultValue=1, parameterName=HeightRepeat})
  n40 = Add(A=Subtract(A=ObjectPositionWS(), B=Add(A=@n38.XYZ, B=@n39)), B=@n38.XYZ)
  n41 = ComponentMask(None=@n40) {b=true}
  n42 = Sine(None=@n41)
  n43 = Cosine(None=@n41)
  Add(
   A=Multiply(
    A=MaterialFunctionCall(R=@n42, G=@n43) {materialFunction=&AppendMany}.RGB
    B=ScalarParameter() {defaultValue=5000, parameterName=Radius}
   )
   B=@n39
  )
  ComponentMask(None=TextureCoordinate() {coordinateIndex=1}) {g=true}
  Divide(A=@n40, B=ScalarParameter() {defaultValue=1, parameterName=Frequency}) {constB=100}
  MaterialFunctionCall() {materialFunction=&ObjectPivotPoint}
  MaterialFunctionCall() {materialFunction=&Near_Camera_Fade}
  n44 = ScalarParameter() {defaultValue=4, parameterName=Tiling}
  n45 = Multiply(A=TextureCoordinate(), B=@n44)
  n46 = ScalarParameter() {defaultValue=4, parameterName=Tiling}
  n47 = Frac(None=Multiply(A=TextureCoordinate(), B=@n46))
  n48 = VertexColor()
  n49 = ScalarParameter() {defaultValue=25, parameterName=RotationSpeed}
  n50 = ScalarParameter() {defaultValue=-25, parameterName=RotationSpeedReverse}
  n51 = VertexColor()
  n52 = Multiply(
   A=Multiply(A=Time(), B=LinearInterpolate(Alpha=@n48.G) {constA=0.5})
   B=Add(A=ScalarParameter() {defaultValue=1, parameterName=RotationSpeed_Min}, B=LinearInterpolate(A=@n49, B=@n50, Alpha=@n51.R))
  )
  n53 = Add(A=Divide(A=Floor(None=@n45), B=@n44), B=Divide(A=Rotator(Coordinate=@n47, Time=@n52), B=@n46))
  n54 = Constant2Vector() {g=0.5, r=0.5}
  n55 = Constant() {r=0.45}
  n56 = ConstantBiasScale(None=Sine(None=Add(A=@n35) {constB=-0.25}))
  Multiply(
   A=Multiply(
    A=TextureSample(UVs=@n53) {texture=/Game/Effects/Fort_Effects/Textures/Rocks/T_Rock_Stone_Sprite_D}.A
    B=SphereMask(A=@n47, B=@n54, Radius=@n55)
   )
   B=Power(Base=LinearInterpolate(Alpha=@n56)) {constExponent=1}
  )
  Multiply(
   A=VectorParameter() {parameterName=Color, defaultValue=<rgba 0.25 0.25 0.25 1>}.RGB
   B=ScalarParameter() {defaultValue=0.5, parameterName=Darkness}
  )
  ObjectPositionWS()
  Panner()
  ScalarParameter()
  ScalarParameter() {parameterName=Scale distance}
  TextureCoordinate() {coordinateIndex=1}
  TextureCoordinate()
  TextureSample(UVs=@n53) {samplerType=SAMPLERTYPE_Normal, texture=/Game/Effects/Fort_Effects/Textures/Rocks/T_Rock_Stone_Sprite_N}
 ]
}
