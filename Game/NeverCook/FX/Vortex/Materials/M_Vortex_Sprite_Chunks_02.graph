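# Masked lit material for vortex debris, drawing rock chunk sprites that circle and travel vertically in a loop.
# World position offset combines the Sprite function, sized by camera-distance fade and vertex green, with sine/cosine motion and a radius-scaled vertical loop phased by UV1.
# Each tile of the rock atlas rotates in place at speeds varied by vertex color, with a sphere mask trimming alpha to the cell.
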Material /Game/NeverCook/FX/Vortex/Materials/M_Vortex_Sprite_Chunks_02
graph {
 properties {
  blendMode=BLEND_Masked
  editorOnlyData="<self>:M_Vortex_Sprite_Chunks_01EditorOnlyData"
  pixelDepthOffsetMode=PDOM_Legacy
  usedShadingModels=Default Lit
 }
 nodes [
  n1 = LinearInterpolate(
   A=ComponentMask(None=VectorParameter() {parameterName=SpriteSizeNear, defaultValue=<rgba 250 250 500 500>}.RGB) {g=true, r=true}
   B=ComponentMask(None=VectorParameter() {parameterName=SpriteSize, defaultValue=<rgba 800 800 500 500>}.RGB) {g=true, r=true}
   Alpha=MaterialFunctionCall(
    Fade_Start_Distance=ScalarParameter() {defaultValue=25000, parameterName="Fade_Start_Distance (S)"}
    Fade_End_Distance=ScalarParameter() {defaultValue=5000, parameterName="Fade_End_Distance (S)"}
   ) {materialFunction=&MF_Distance_SyncedFade_FX}.Result
  )
  n2 = LinearInterpolate(Alpha=VertexColor().G) {constA=0.5}
  n3 = Multiply(A=Time(), B=ScalarParameter() {defaultValue=1, parameterName=CircleSpeed})
  n4 = Sine(None=@n3)
  n5 = AppendVector(A=@n4, B=Cosine(None=@n3))
  n6 = ScalarParameter() {defaultValue=0.6, parameterName=MinVerticalSpeedMult}
  n7 = Constant() {r=0.5}
  n8 = Multiply(A=ScalarParameter() {defaultValue=0.6, parameterName=MinVerticalSpeedMult_02}, B=@n7)
  n9 = Multiply(
   A=ObjectRadius()
   B=LinearInterpolate(A=Multiply(A=@n6, B=@n7), B=StaticSwitchParameter(True=@n8, False=@n7) {parameterName=MoveAllDownWards?}, Alpha=ConstantBiasScale(None=@n4)) {constA=0.5, constB=0.5}
  ) {constB=0.5}
  n10 = Frac(None=Divide(A=Time(), B=ScalarParameter() {defaultValue=250, parameterName=LoopDuration}))
  n11 = StaticSwitchParameter(True=OneMinus(None=@n10), False=@n10) {parameterName=ReverseDirection?}
  n12 = TextureCoordinate() {coordinateIndex=1}
  n13 = Multiply(A=Subtract(A=ComponentMask(None=@n12) {g=true}) {constB=0.5}) {constB=-2}
  n14 = LinearInterpolate(A=Multiply(A=@n9) {constB=-1}, B=@n9, Alpha=Frac(None=Add(A=@n11, B=@n13)))
  Add(
   A=MaterialFunctionCall!(Center Location=WorldPosition().XYZ, 0-1 UVs, XY scale=Multiply(A=@n1, B=@n2), OverridePivot, "Normals (Optional)") {materialFunction=&Sprite}.World position offset
   B=Multiply(A=ScalarParameter() {defaultValue=1, parameterName=Param_1}, B=AppendVector(A=@n5, B=@n14))
  )
  MaterialFunctionCall() {materialFunction=&Near_Camera_Fade}
  MaterialFunctionCall!(Time=Time(), Divisor=ScalarParameter() {defaultValue=1, parameterName=Param_2}, Mod=ScalarParameter() {defaultValue=2}) {materialFunction=/Game/Effects/Fort_Effects/Materials/_Functions/MF_ModuloTime}
  n15 = ScalarParameter() {defaultValue=4, parameterName=Tiling}
  n16 = Multiply(A=TextureCoordinate(), B=@n15)
  n17 = ScalarParameter() {defaultValue=4, parameterName=Tiling}
  n18 = Frac(None=Multiply(A=TextureCoordinate(), B=@n17))
  n19 = VertexColor()
  n20 = ScalarParameter() {defaultValue=2, parameterName=RotationSpeed}
  n21 = ScalarParameter() {defaultValue=-2, parameterName=RotationSpeedReverse}
  n22 = VertexColor()
  n23 = Multiply(
   A=Multiply(A=Time(), B=LinearInterpolate(Alpha=@n19.G) {constA=0.5})
   B=Add(A=ScalarParameter() {defaultValue=0.1, parameterName=RotationSpeed_Min}, B=LinearInterpolate(A=@n20, B=@n21, Alpha=@n22.R))
  )
  n24 = Add(A=Divide(A=Floor(None=@n16), B=@n15), B=Divide(A=Rotator(Coordinate=@n18, Time=@n23), B=@n17))
  n25 = TextureSample(UVs=@n24) {texture=/Game/Effects/Fort_Effects/Textures/Rocks/T_Rock_Stone_Sprite_D}
  Multiply(A=@n25.A, B=SphereMask(A=@n18, B=Constant2Vector() {g=0.5, r=0.5}, Radius=Constant() {r=0.45}))
  Multiply(A=@n25.RGB, B=ScalarParameter() {defaultValue=0.5, parameterName=Darkness})
  ScalarParameter() {parameterName=Scale distance}
  TextureCoordinate()
  TextureSample(UVs=@n24) {samplerType=SAMPLERTYPE_Normal, texture=/Game/Effects/Fort_Effects/Textures/Rocks/T_Rock_Stone_Sprite_N}
 ]
}
