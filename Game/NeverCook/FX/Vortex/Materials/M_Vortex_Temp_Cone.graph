Material /Game/NeverCook/FX/Vortex/Materials/M_Vortex_Temp_Cone
graph {
 properties {blendMode=BLEND_Masked, pixelDepthOffsetMode=PDOM_Legacy, usedShadingModels=Default Lit}
 nodes [
  n1 = AppendVector(A=ScalarParameter() {defaultValue=5, parameterName=X}, B=ScalarParameter() {defaultValue=10, parameterName=Y})
  n2 = TextureCoordinate()
  n3 = ScalarParameter() {defaultValue=0.01, parameterName=X_WobbleSpeed}
  n4 = ScalarParameter() {defaultValue=0.1, parameterName=Y_WobbleSpeed}
  n5 = TextureSample(UVs=Panner(Coordinate=Multiply(A=@n1, B=@n2), Speed=AppendVector(A=@n3, B=@n4)) {speedY=0.4}) {samplerType=SAMPLERTYPE_Normal, texture=/Game/Global/Textures/Default/Noise/T_SwirlNoise_N}
  n6 = Divide(A=ComponentMask(None=WorldPosition().XYZ) {b=true}, B=ScalarParameter() {defaultValue=50000, parameterName=SpiralFrequency})
  n7 = Multiply(A=Time(), B=ScalarParameter() {defaultValue=-0.05, parameterName=SpiralSpeed})
  n8 = VectorParameter() {parameterName=Direction, defaultValue=<rgba 1 0 0 1>}
  n9 = Multiply(A=@n8.RGB, B=ScalarParameter() {defaultValue=2500, parameterName=SpiralSize})
  n10 = Multiply(A=@n8.RGB, B=ScalarParameter() {defaultValue=500, parameterName=SpiralSizeLow})
  n11 = ComponentMask(None=TextureCoordinate() {coordinateIndex=1}) {g=true}
  Add(
   A=Multiply(A=@n5.RGB, B=ScalarParameter() {defaultValue=50, parameterName=Distort})
   B=RotateAboutAxis(
    NormalizedRotationAxis=VectorParameter() {parameterName=RotationAxis, defaultValue=<rgba 0 0 1 1>}.RGB
    RotationAngle=Add(A=@n6, B=@n7)
    PivotPoint=Constant3Vector()
    Position=LinearInterpolate(A=@n9, B=@n10, Alpha=@n11)
   )
  )
  Add(A=TextureCoordinate(), B=Multiply(A=@n5.RGB, B=ScalarParameter() {defaultValue=0.25, parameterName=WiggleIntensity}))
  Clamp(None=MaterialFunctionCall(Normal=@n5.RGB, Flatness=ScalarParameter() {defaultValue=0.95, parameterName="Flatness (S)"}) {materialFunction=&FlattenNormal}.Result)
  ComponentMask(None=@n5.RGB) {g=true, r=true}
  n12 = ComponentMask(None=@n5.G) {g=true}
  n13 = TextureCoordinate() {vTiling=0.8}
  n14 = OneMinus(None=@n13)
  ComponentMask(None=LinearInterpolate(A=Multiply(A=@n12, B=@n14), Alpha=@n14)) {g=true}
  n15 = Multiply(A=Time(), B=ScalarParameter() {defaultValue=1, parameterName=RotationSpeed})
  n16 = Add(A=@n15, B=Frac(None=Sine(None=@n15)))
  n17 = MaterialFunctionCall(R=Sine(None=@n16), B=Cosine(None=@n16)) {materialFunction=&AppendMany}
  n18 = ScalarParameter() {defaultValue=250, parameterName=Multiplier}
  n19 = VectorParameter() {parameterName=Offset, defaultValue=<rgba 0 0 0 1>}
  n20 = Divide(
   A=Distance(
    A=WorldPosition() {worldPositionShaderOffset=WPT_ExcludeAllShaderOffsets}.XYZ
    B=TransformPosition(None=@n19.RGB) {transformType=TRANSFORMPOSSOURCE_World}
   )
   B=ScalarParameter() {defaultValue=150000, parameterName=Rad}
  )
  LinearInterpolate(
   A=Constant3Vector()
   B=Transform(None=Multiply(A=@n17.RGB, B=@n18)) {transformSourceType=TRANSFORMSOURCE_Local}
   Alpha=Power(Base=Saturate(None=@n20), Exp=ScalarParameter() {defaultValue=2, parameterName=LinearFalloffPower})
  )
  n21 = TextureCoordinate()
  n22 = CurveAtlasRowParameter(CurveTime=ComponentMask(None=TextureCoordinate()) {g=true}) {
   atlas=&AthenaCurveAtlas01
   curve=/Game/Athena/LinearColorCurves/Mello_Rainbow
   defaultValue=52
   parameterName=Param_1
  }
  LinearInterpolate(
   A=MaterialFunctionCall(Hue Shift Percentage=ComponentMask(None=@n21) {g=true}, Texture=@n22) {materialFunction=&HueShift}.Result
   B=ScalarParameter() {defaultValue=1}
   Alpha=ScalarParameter() {defaultValue=0.7, parameterName=BlendColorAndCurve}
  )
  Multiply(A=@n15) {constB=0.45}
  Multiply(A=@n13, B=@n13)
  ObjectPositionWS()
  ScalarParameter() {defaultValue=15, parameterName=Specular}
  ScalarParameter() {defaultValue=0.25, parameterName=Roughness}
  ScalarParameter() {defaultValue=1, parameterName=Metallic}
 ]
}
